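; Projectile actor for the Jolly Rascal consumable's extra ammo drop, which bursts into ammo pickups when it hits something.
; BeginPlay gives the rotating movement a random signed spin on each axis.
; On hit it stops simulating, and the server picks loot and spawns pickups at an offset, scaling counts by an ammo scaler row.
; It then plays a destroy GameplayCue, hides via a replicated flag for replays, fades audio, and destroys itself after a delay.
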(event Custom|SetRotation
  (params (Rotation%20Rate%20X%20%28Roll%29 :type Float%20%28double-precision%29) (Rotation%20Rate%20Y%20%28Pitch%29 :type Float%20%28double-precision%29) (Rotation%20Rate%20Z%20%28Yaw%29 :type Float%20%28double-precision%29))
  ;[1 1 1] Set Random Rotation On Begin
  (Variables|/Script/Engine.RotatingMovementComponent|SetRotationRate_Roll (Math|Rotator|MakeRotator Rotation%20Rate%20X%20%28Roll%29 Rotation%20Rate%20Y%20%28Pitch%29 Rotation%20Rate%20Z%20%28Yaw%29) (Variables|Default|GetRotatingMovement)))

(event EventBeginPlay
  ;[1 1 1] Set Random Rotation On Begin
  (|CallParentFunction)
  (CallFunction|SetRotation (CallFunction|RandomSpin) (CallFunction|RandomSpin) (CallFunction|RandomSpin)))

(event Custom|HideAndKill
  ;[1 1 1] Play GC, Hide Actor and Kill
  (Ability|ExecuteGameplayCueReplicated (Variables|GCs|GetGC_Destroy))
  (Development|SetHiddeninGame (Variables|Default|GetMesh) true false)
  (Variables|Default|SetHiddenInGame true)
  (Components|Activation|Deactivate (Variables|Default|GetTrail_Niagara))
  (Audio|Components|Audio|FadeOut (Variables|Default|GetAudio) 0.25 0.0 "Linear" :!class /Script/Engine.AudioComponent)
  (Utilities|FlowControl|Delay 1.0)
  (Actor|DestroyActor))

(event Collision|EventHit
  (params (MyComp :type Primitive%20Component%20Object%20Reference) (Other :type Actor%20Object%20Reference) (OtherComp :type Primitive%20Component%20Object%20Reference) (bSelfMoved :type Boolean) (HitLocation :type Vector :default "0, 0, 0") (HitNormal :type Vector :default "0, 0, 0") (NormalImpulse :type Vector :default "0, 0, 0") (Hit :type Hit%20Result%20Structure%20%28by%20ref%29))
  ;[1 1 1] On Hit Spawn items and get rid of actor
  (Game|Components|ProjectileMovement|StopSimulating (Variables|Default|GetFortProjectileMovement) (Variables|Default|GetNullHit))
  (CallFunction|SpawnItems)
  (CallFunction|HideandKill))

(event Custom|Spawn%20Items
  ;[1 1 1] Spawn Items from array
  (if (Networking|IsServer)
    (Development|PrintString "Spawn Items")
    (bind (_outloottodrop _returnvalue) (Loot|PickLootDrops (Variables|Default|GetItemsToDrop) 1 -1 "(GameplayTags=)"))
    (if _returnvalue
      (for _array_element _outloottodrop
        (bind (_array_element_itemdefinition _array_element_itemguid _array_element_count) (Class|FortItemFunctionLibrary|BreakFortItemEntry _array_element))
        (bind _asfort_world_item_definition (Utilities|Casting|CastToFortWorldItemDefinition _array_element_itemdefinition))
        (Inventory|SpawnPickupInWorld _asfort_world_item_definition (Math|Float|Truncate (* _array_element_count (ScalableFloat|GetValueatLevel (Variables|Rows|GetRows_AmmoScaler)))) (+ (Transformation|GetActorLocation) (Variables|Default|GetSpawnPositionOffset)) (Transformation|GetActorForwardVector))))))

(fn OnRep_HiddenInGame
  ;[1 1 1] Rep the hide for replays
  (Development|SetHiddeninGame (Variables|Default|GetMesh) (Variables|Default|GetHiddenInGame) false))

(fn RandomSpin
  (return (select (Math|Random|RandomBoolwithWeight 0.5) (* (Math|Random|RandomFloatinRange (LocalVariables|GetMinSpin) (LocalVariables|GetMaxSpin)) -1.0) (Math|Random|RandomFloatinRange (LocalVariables|GetMinSpin) (LocalVariables|GetMaxSpin)))))

(fn ConstructionScript
  (|CallParentFunction))
