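; Jolly Rascal extra-drop projectile actor that spins through the air and drops wood, stone or metal pickups on impact.
; BeginPlay picks the item, count and mesh from the resource type; the server sends random spin through RepRotation and line-traces ahead for water.
; On hit it plays a water splash if water was found, spawns the pickup server-side, then fires a destroy cue, hides and destroys itself.
; A RepNotify hidden flag keeps replays correct.
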(event Custom|RepRotation
  (params (Rotation%20Rate%20X%20%28Roll%29 :type Float%20%28double-precision%29) (Rotation%20Rate%20Y%20%28Pitch%29 :type Float%20%28double-precision%29) (Rotation%20Rate%20Z%20%28Yaw%29 :type Float%20%28double-precision%29))
  ;[1 1 1] Rep Rotation
  (Variables|/Script/Engine.RotatingMovementComponent|SetRotationRate_Roll (Math|Rotator|MakeRotator Rotation%20Rate%20X%20%28Roll%29 Rotation%20Rate%20Y%20%28Pitch%29 Rotation%20Rate%20Z%20%28Yaw%29) (Variables|Default|GetRotatingMovement)))

(event EventBeginPlay
  ;[1 1 1] Begin Play - Set up the mesh and material type. Also set the spin values via RPC
  (CallFunction|SelectResourceType)
  (CallFunction|SelectMesh)
  (|CallParentFunction)
  (if (Networking|IsServer)
    (CallFunction|RepRotation (CallFunction|RandomSpin) (CallFunction|RandomSpin) (CallFunction|RandomSpin))
    (CallFunction|WaterHitCheck)))

(event Custom|HideAndKill
  ;[1 1 1] Play GC, Hide Actor and Kill
  (Ability|ExecuteGameplayCueReplicated (Variables|Default|GetGC_Destroy))
  (Collision|SetActorEnableCollision false)
  (Development|SetHiddeninGame (Variables|Default|GetMesh) true false)
  (Variables|Default|SetHiddenInGame true)
  (Components|Activation|Deactivate (Variables|Default|GetTrail_Niagara))
  (Audio|Components|Audio|FadeOut (Variables|Default|GetAudio) 0.25 0.0 "Linear" :!class /Script/Engine.AudioComponent)
  (Utilities|FlowControl|Delay 1.0)
  (Actor|DestroyActor))

(event Collision|EventHit
  (params (MyComp :type Primitive%20Component%20Object%20Reference) (Other :type Actor%20Object%20Reference) (OtherComp :type Primitive%20Component%20Object%20Reference) (bSelfMoved :type Boolean) (HitLocation :type Vector :default "0, 0, 0") (HitNormal :type Vector :default "0, 0, 0") (NormalImpulse :type Vector :default "0, 0, 0") (Hit :type Hit%20Result%20Structure%20%28by%20ref%29))
  (bind _water_impact_point (Variables|Water|GetWater%20Impact%20Point))
  ;[1 1 1] On Hit Spawn items and get rid of actor
  (Utilities|FlowControl|Sequence
    (:then_0
      (if (Variables|Water|GetFound%20Water)
        (bind _returnvalue (Niagara|SpawnSystematLocation "/Game/Effects/Niagara/Systems/Enviroment/Water/NS_WaterSplash_Medium.NS_WaterSplash_Medium" _water_impact_point))
        (Niagara|SetNiagaraVariable(Vector3) _returnvalue "WaterPlaneLocation" _water_impact_point)
        (Niagara|SetNiagaraVariable(Vector3) _returnvalue "WaterPlaneNormal" (Variables|Water|GetWater%20Surface%20Normal))
        (Niagara|SetNiagaraVariable(Float) _returnvalue "WaterDepth" (Variables|Water|GetWater%20Depth))
        (Audio|PlaySoundatLocation "/Game/Athena/Sounds/Weapons/Explosions/Expl_Debris_Water_Small_Cue.Expl_Debris_Water_Small_Cue" (Variables|Water|GetWater%20Impact%20Point))))
    (:then_1
      (Game|Components|ProjectileMovement|StopSimulating (Variables|Default|GetFortProjectileMovement) (Variables|Default|GetNullHit))
      (CallFunction|SpawnItems)
      (CallFunction|HideandKill))))

(event Custom|Spawn%20Items
  (bind _itemcounttospawn (Variables|ItemSpawn|GetItemCountToSpawn))
  ;[1 1 1] Spawn Items from array
  (if (Networking|IsServer)
    (Development|PrintString "Spawn Items")
    (if (> _itemcounttospawn 0)
      (Inventory|SpawnPickupInWorld (Variables|ItemSpawn|GetItemDefToSpawn) _itemcounttospawn (+ (Transformation|GetActorLocation) (Variables|ItemSpawn|GetSpawnPositionOffset)) (Transformation|GetActorForwardVector)))))

(fn OnRep_HiddenInGame
  ;[1 1 1] Rep the hide for replays
  (Development|SetHiddeninGame (Variables|Default|GetMesh) (Variables|Default|GetHiddenInGame) false))

(fn RandomSpin
  (return (select (Math|Random|RandomBoolwithWeight 0.5) (* (Math|Random|RandomFloatinRange (Variables|Default|GetMinSpin) (Variables|Default|GetMaxSpin)) -1.0) (Math|Random|RandomFloatinRange (Variables|Default|GetMinSpin) (Variables|Default|GetMaxSpin)))))

(fn SelectMesh
  (switch Utilities|FlowControl|Switch|SwitchonEFortResourceType (Variables|ItemSpawn|GetResourceType)
    (:Wood
      _)
    (:Stone
      (CallFunction|SetMesh (Variables|MeshPick|GetStoneMeshes) (Variables|MeshPick|GetMap_StoneTransform)))
    (:Metal
      (CallFunction|SetMesh (Variables|MeshPick|GetMetalMeshes) (Variables|MeshPick|GetMap_MetalTransform)))
    (:Permanite
      _)
    (:None
      _))
  (CallFunction|SetMesh (Variables|MeshPick|GetWoodMeshes) (Variables|MeshPick|GetMap_WoodTransform)))

(fn SelectResourceType
  (switch Utilities|FlowControl|Switch|SwitchonEFortResourceType (Variables|ItemSpawn|GetResourceType)
    (:Wood
      _)
    (:Stone
      (Variables|ItemSpawn|SetItemDefToSpawn "/Game/Items/ResourcePickups/StoneItemData.StoneItemData")
      (Variables|ItemSpawn|SetItemCountToSpawn (Math|Float|Truncate (ScalableFloat|GetValueatLevel (Variables|Rows|GetRow_StoneToSpawn)))))
    (:Metal
      (Variables|ItemSpawn|SetItemDefToSpawn "/Game/Items/ResourcePickups/MetalItemData.MetalItemData")
      (Variables|ItemSpawn|SetItemCountToSpawn (Math|Float|Truncate (ScalableFloat|GetValueatLevel (Variables|Rows|GetRow_MetalToSpawn)))))
    (:Permanite
      _)
    (:None
      _))
  (Variables|ItemSpawn|SetItemDefToSpawn "/Game/Items/ResourcePickups/WoodItemData.WoodItemData")
  (Variables|ItemSpawn|SetItemCountToSpawn (Math|Float|Truncate (ScalableFloat|GetValueatLevel (Variables|Rows|GetRow_WoodToSpawn)))))

(fn SetMesh
  (params (Array :type Array%20of%20Static%20Mesh%20Object%20References) (TargetMap :type Map%20of%20Integers%20to%20Transforms))
  (bind _mesh (Variables|Default|GetMesh))
  (Components|StaticMesh|SetStaticMesh _mesh (Utilities|Array|Get(acopy) Array (select (> (Variables|MeshPick|GetMeshIndex) (- (Utilities|Array|Length Array) 1)) (- (Utilities|Array|Length Array) 1) (Variables|MeshPick|GetMeshIndex))))
  (Transformation|SetRelativeTransform _mesh (Utilities|Map|Find TargetMap (Variables|MeshPick|GetMeshIndex)) false false))

(fn ConstructionScript
  (|CallParentFunction))

(fn WaterHitCheck
  (bind (_outhit _returnvalue) (Collision|LineTraceByChannel (Transformation|GetActorLocation) (+ (Transformation|GetActorLocation) (Variables|Water|GetLineTraceOffset)) "TraceTypeQuery12"))
  (if _returnvalue
    (bind _aswater_body (Utilities|Casting|CastToWaterBody (.HitActor (Collision|BreakHitResult _outhit))))
    (Variables|Water|SetWater%20Impact%20Point (WaterBody|GetWaterSurfaceInfoatLocation (Variables|/Script/Water.WaterBody|GetWaterBodyComponent _aswater_body) (.ImpactPoint (Collision|BreakHitResult _outhit)) true))
    (Variables|Water|SetWater%20Surface%20Normal (.OutWaterSurfaceNormal (WaterBody|GetWaterSurfaceInfoatLocation (Variables|/Script/Water.WaterBody|GetWaterBodyComponent _aswater_body) (.ImpactPoint (Collision|BreakHitResult _outhit)) true)))
    (Variables|Water|SetWater%20Depth (.OutWaterDepth (WaterBody|GetWaterSurfaceInfoatLocation (Variables|/Script/Water.WaterBody|GetWaterBodyComponent _aswater_body) (.ImpactPoint (Collision|BreakHitResult _outhit)) true)))
    (Variables|Water|SetFound%20Water true)))
