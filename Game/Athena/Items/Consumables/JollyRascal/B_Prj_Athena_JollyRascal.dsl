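; Jolly Rascal thrown projectile that spawns a junk spawner building actor above where it hits or enters water.
; Each event, once and server-only, sphere-traces from an offset toward the projectile to pick the spawn point, adds another offset and randomizes rotation.
; FXandSounds stretches a Niagara lightning effect by distance to the spawn point and shows a sound indicator for all affiliations.
; HideAndKill then hides the projectile, disables collision and sets a short lifespan.
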(event Custom|FXandSounds
  (params (Location :type Vector))
  (bind _returnvalue (Transformation|GetActorLocation))
  ;[1 1 1] FX and audio
  (bind _returnvalue_2 (Niagara|SpawnSystematLocation "/Game/Athena/Items/Consumables/JollyRascal/FX/NS_Jolly_Launch.NS_Jolly_Launch" _returnvalue))
  (Utilities|IsValid _returnvalue_2
    (:Is%20Valid
      (Effects|Components|ParticleSystem|SetVectorParameter _returnvalue_2 "Scale" (Math|Vector|MakeVector (/ (Math|Vector|VectorLength (- _returnvalue Location)) 1e+03) 1.0 1.0)) ; Extend lightning to portal spawn
      (SoundIndicator|ShowSoundIndicatorforDuration (Variables|Default|GetFortSoundIndicator) (Transformation|GetActorLocation) (Variables|/Script/Engine.AudioComponent|GetSound (Variables|Default|GetGrenadeFuse_AudioComponent)) 1.0 1e+01 0)
      (Networking|ForceNetUpdate))))

(event OnEnterWaterEvent(WaterInteractionComponent)
  (params (WaterBody :type Fort%20Water%20Body%20Actor%20Object%20Reference) (WaterInteractionComponent :type Fort%20Water%20Interaction%20Component%20Object%20Reference) (bIsFirstBody :type Boolean))
  (bind _returnvalue (Transformation|GetActorLocation))
  (bind _spawnoffset (Variables|Default|GetSpawnOffset))
  ;[1 1 1] On Hit Check Above the Hit and spawn the top.
  (Utilities|FlowControl|DoOnce false
    (:Completed
      (if (Networking|IsServer)
        (bind (_outhit _returnvalue_2) (Collision|SphereTraceByChannel (+ _returnvalue (Variables|Default|GetTraceOffset)) _returnvalue (Variables|Default|GetTraceOverlapRadius) "TraceTypeQuery10"))
        (CallFunction|FXandSounds (+ (select _returnvalue_2 (.Location (Collision|BreakHitResult _outhit)) _returnvalue) _spawnoffset))
        (Game|SpawnBuildingActorfromClass (Variables|Default|GetJunkSpawnerClass) (Math|Transform|MakeTransform (+ (select _returnvalue_2 (.Location (Collision|BreakHitResult _outhit)) _returnvalue) _spawnoffset) (Math|Rotator|MakeRotator 0.0 0.0 (Math|Random|RandomFloatinRange 0.0 3.6e+02)) "1.000000,1.000000,1.000000") (Game|GetInstigator) false (Game|GetInstigator))
        (CallFunction|HideandKill)
        ;;[0 0.8957 1] Detached due to bug killing player on impact.
        (Delivery|DoExplosion)))))

(event Collision|EventHit
  (params (MyComp :type Primitive%20Component%20Object%20Reference) (Other :type Actor%20Object%20Reference) (OtherComp :type Primitive%20Component%20Object%20Reference) (bSelfMoved :type Boolean) (HitLocation :type Vector :default "0, 0, 0") (HitNormal :type Vector :default "0, 0, 0") (NormalImpulse :type Vector :default "0, 0, 0") (Hit :type Hit%20Result%20Structure%20%28by%20ref%29))
  (bind _returnvalue (Transformation|GetActorLocation))
  (bind _spawnoffset (Variables|Default|GetSpawnOffset))
  ;[1 1 1] On Hit Check Above the Hit and spawn the top.
  (Utilities|FlowControl|DoOnce false
    (:Completed
      (if (Networking|IsServer)
        (bind (_outhit _returnvalue_2) (Collision|SphereTraceByChannel (+ _returnvalue (Variables|Default|GetTraceOffset)) _returnvalue (Variables|Default|GetTraceOverlapRadius) "TraceTypeQuery10"))
        (CallFunction|FXandSounds (+ (select _returnvalue_2 (.Location (Collision|BreakHitResult _outhit)) _returnvalue) _spawnoffset))
        (Game|SpawnBuildingActorfromClass (Variables|Default|GetJunkSpawnerClass) (Math|Transform|MakeTransform (+ (select _returnvalue_2 (.Location (Collision|BreakHitResult _outhit)) _returnvalue) _spawnoffset) (Math|Rotator|MakeRotator 0.0 0.0 (Math|Random|RandomFloatinRange 0.0 3.6e+02)) "1.000000,1.000000,1.000000") (Game|GetInstigator) false (Game|GetInstigator))
        (CallFunction|HideandKill)
        ;;[0 0.8957 1] Detached due to bug killing player on impact.
        (Delivery|DoExplosion)))))

(event Custom|HideAndKill
  (Rendering|SetVisibility (Variables|Default|GetMesh) false false)
  (Components|Activation|SetActive (Variables|Default|GetFuse_ParticleNiagara) false false)
  (Collision|SetActorEnableCollision false)
  (Actor|SetLifeSpan (Variables|Default|GetLifeAfterHide)))

(fn GetAffiliationsToShowFor
  (return (Utilities|Array|MakeArray "Neutral" "Friendly" "Hostile")))

(fn ShouldShowSoundIndicator
  (params (PlayerController :type Fort%20Player%20Controller%20Object%20Reference))
  (return true))

(fn ConstructionScript
  (|CallParentFunction))
