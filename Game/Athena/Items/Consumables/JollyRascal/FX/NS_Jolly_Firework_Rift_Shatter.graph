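# Niagara burst effect for a rift shatter in the JollyRascal consumable's firework FX.
# A stretched flare, round flare and expanding shockwave mesh form the flash, with brief flickering electric-arc meshes as residue.
# Three near-identical mesh fragment emitters and GPU sparkle sprites shoot sharply downward from inverted hemispheres, slowed by drag.
# User colors FXColorA, B and C tint arcs, flares with sparkles, and fragments respectively, boosted by HDR curves.
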NiagaraSystem /Game/Athena/Items/Consumables/JollyRascal/FX/NS_Jolly_Firework_Rift_Shatter
dependencies {
 usedDataInterfaces [&NiagaraDataInterfaceCurve, &NiagaraDataInterfaceVectorCurve]
 usedDynamicInputs [&ConvertVectorToPosition]
 usedModules [&SystemState]
}
emitters {
 ElectricArcsResidual {
  emitterUpdateScript [
   &EmitterState(Life Cycle Mode=::NewEnumerator1, Loop Behavior=::NewEnumerator1, Loop Duration=0.5, UseLoopDelay=true)
   &SpawnBurst_Instantaneous(Spawn Count=6, Spawn Time=0.1, Use Spawn Probability=false, UseLoopCountLimit=false)
   &SpawnRate(
    SpawnRate=&Multiply_Float(
     A=30
     B=&FloatFromCurve(FloatCurve=DataInterface() {propertyValues {Curve {keys [<at 0 1 RCIM_Cubic>,<at 1 0.5>]}}}, CurveIndex=$Emitter.NormalizedLoopAge)
    )
    Use Spawn Probability=false
   )
  ]
  particleSpawnScript [
   &InitializeParticle(
    Color=&MultiplyLinearColor(
     Linear Color=$User.FXColorA
     Scale Factor=&MakeFloat4FromVectorAndFloat(
      "Vector (RGB)"=&VectorFromFloat(Value=&RandomRangeFloat(Minimum=5, Maximum=12))
      "Float (Alpha)"=1
     )
    )
    Lifetime=&RandomRangeFloat(Minimum=0.1, Maximum=0.25, Evaluation Type=::NewEnumerator1)
    Mesh Scale=&RandomRangeVector(Minimum=<10 10 10>, Maximum=<24 18 18>, Evaluation Type=::NewEnumerator1)
    Mesh Scale Mode=::NewEnumerator4
    Position Offset=<0 0 -125>
    UsePositionOffset=true
    Write Material Random=true
   )
   &DynamicMaterialParameters(
    Index 0 Param 1=&RandomRangeFloat(Minimum=2, Maximum=3, Evaluation Type=::NewEnumerator1)
    Index 0 Param 3=&RandomRangeFloat(Minimum=0.2, Maximum=2, Evaluation Type=::NewEnumerator1)
    Index 0 Param 4=&RandomRangeFloat(Minimum=0, Maximum=1, Evaluation Type=::NewEnumerator1)
    Module.Index0Param0Write=true
    Module.Index0Param1Write=false
    Module.Index0Param2Write=true
    Module.Index0Param3Write=true
   )
   /Niagara/Modules/CascadeConversion/CascadeConversion_InitialMeshRotation(Rotation:Vector3f=&RandomRangeVector(Minimum=<0 0 0>, Maximum=<1 0 0>, Evaluation Type=::NewEnumerator1))
   &ShapeLocation(
    Cylinder Height=20
    Cylinder Radius=25
    MaskSpawnGroup=false
    Override Seed=false
    Rotation Axis=<0 0 1>
    Rotation Mode=::NewEnumerator1
    Shape Primitive=::NewEnumerator1
    Surface Only=false
    UseOwnerScale=true
   )
   &AddVelocity(Velocity=&Multiply_VectorByFloat(Vector=&Multiply_Vector(A=$Particles.ShapeLocation.ShapeVector, B=<1 1 0>), Float=5))
   &InitialMeshOrientation(
    Mesh Orientation Mode=::NewEnumerator7
    Rotation=&RandomRangeVector(Minimum=<0 0 0>, Maximum=<0 1 0>)
    Use Rotation Vector=true
   )
  ]
  particleUpdateScript [
   &ParticleState()
   &ScaleMeshSize(Scale Factor=&VectorFromCurve(VectorCurve=DataInterface() {propertyValues {
       XCurve {keys [<at 0.0033 0.5 RCIM_Constant>,<at 0.2054 0.7619 RCIM_Constant>,<at 0.4101 0.8575 RCIM_Constant>,<at 0.6967 1>]}
       YCurve {keys [<at 0.0033 0.6004 RCIM_Constant>,<at 0.2054 0.6716 RCIM_Constant>,<at 0.4101 0.3646 RCIM_Constant>,<at 0.6967 0.2468>]}
       ZCurve {keys [<at 0.0033 0.7767 RCIM_Constant>,<at 0.2054 0.5215 RCIM_Constant>,<at 0.4101 0.4457 RCIM_Constant>,<at 0.6967 0.6282>]}
      }}))
   &DynamicMaterialParameters(
    Index 0 Param 2=&FloatFromCurve(FloatCurve=DataInterface() {propertyValues {Curve {keys [{arriveTangent=-1.8894, interpMode=RCIM_Cubic, time=0, value=1}, <at 1 -0.25>]}}})
    Module.Index0Param0Write=false
    Module.Index0Param1Write=true
    Module.Index0Param2Write=false
    Module.Index0Param3Write=false
   ) {module=DynamicMaterialParameters001}
   &ScaleColor(
    Scale RGB=&VectorFromFloat(Value=&FloatFromCurve(FloatCurve=DataInterface() {propertyValues {Curve {keys [<at 0 50 RCIM_Cubic>, {arriveTangent=-45, interpMode=RCIM_Cubic, leaveTangent=-45, time=0.5, value=30}, <at 1 5>]}}}))
    ScaleA=true
    ScaleRGB=true
   )
   &SolveForcesAndVelocity(Clamp Velocity=false, Limit Acceleration=false)
  ]
  rendererClasses [&NiagaraMeshRendererProperties]
  renderers [{rendererClass=&NiagaraMeshRendererProperties, rendererIndex=0}]
 }
 Endflare {
  emitterUpdateScript [
   &EmitterState(Life Cycle Mode=::NewEnumerator1, Loop Behavior=::NewEnumerator1, Loop Delay=2, Loop Duration=1, UseLoopDelay=true)
   &SpawnBurst_Instantaneous(Use Spawn Probability=false, UseLoopCountLimit=false)
  ]
  particleSpawnScript [&InitializeParticle(
    Color=&MultiplyLinearColor(
     Linear Color=$User.FXColorB
     Scale Factor=&MakeFloat4FromVectorAndFloat("Vector (RGB)"=&VectorFromFloat(Value=2), "Float (Alpha)"=1)
    )
    Lifetime=0.6
    Position Offset=<0 0 -125>
    Sprite Size=<1600 300>
    Sprite Size Mode=::NewEnumerator4
    UsePositionOffset=true
    Write Material Random=true
   )]
  particleUpdateScript [
   &ParticleState()
   &ScaleSpriteSize(
    Scale Factor=&Vector2DFromFloat(Value=&FloatFromCurve(FloatCurve=DataInterface() {propertyValues {Curve {keys [<at 0 1 RCIM_Cubic>, {interpMode=RCIM_Cubic, leaveTangent=-3.3604, time=1, value=0}]}}}))
    Scale Sprite Size Mode=::NewEnumerator2
   )
   &ScaleColor(
    Scale Alpha=&FloatFromCurve(FloatCurve=DataInterface() {propertyValues {Curve {keys [<at 0 0 RCIM_Cubic>, <at 0.6 1 RCIM_Cubic>, {arriveTangent=-8.5075, leaveTangent=-8.5075, time=1, value=0}]}}})
    Scale RGB=<2 2 2>
    ScaleA=true
    ScaleRGB=true
   )
   &DynamicMaterialParameters(
    Index 0 Param 1=&FloatFromCurve(FloatCurve=DataInterface() {propertyValues {Curve {keys [<at 0 0.15 RCIM_Cubic>,<at 1 0>]}}})
    Index 0 Param 2=5
    Index 0 Param 3=1
    Index 0 Param 4=1
    Module.Index0Param0Write=true
    Module.Index0Param1Write=true
    Module.Index0Param2Write=true
    Module.Index0Param3Write=true
   )
   &CameraOffset(Camera Offset Amount=40)
  ]
  rendererClasses [&NiagaraSpriteRendererProperties]
  renderers [{rendererClass=&NiagaraSpriteRendererProperties}]
 }
 Frags1 {
  emitterUpdateScript [
   &EmitterState(Life Cycle Mode=::NewEnumerator1, Loop Behavior=::NewEnumerator1, Loop Duration=0.5, UseLoopDelay=true)
   &SpawnBurst_Instantaneous(Spawn Count=5, Use Spawn Probability=false, UseLoopCountLimit=false)
  ]
  particleSpawnScript [
   /SetVariables(Particles.SpawnRotationVector=&RandomRangeVector(Minimum=<0.1 1.5 -0.1>, Maximum=<0.4 2 0.1>, Evaluation Type=::NewEnumerator1)) {isSetParameters=true}
   &InitializeParticle(
    Color=&MultiplyLinearColor(Linear Color=$User.FXColorC, Scale Factor=<10 10 10 1>)
    Lifetime=&RandomRangeFloat(Minimum=1.5, Maximum=2, Evaluation Type=::NewEnumerator1)
    Mesh Scale=&RandomRangeVector(Minimum=<2 2 3>, Maximum=<5 3 3>, Evaluation Type=::NewEnumerator1)
    Mesh Scale Mode=::NewEnumerator4
    Position Offset=<0 0 -50>
    UsePositionOffset=true
    Write Material Random=true
   )
   &AddVelocity(Velocity=&RandomRangeVector(Minimum=<0 0 -3500>, Maximum=<0 0 -1600>, Evaluation Type=::NewEnumerator1))
   &ShapeLocation(
    Hemisphere Distribution=<1 0.5>
    MaskSpawnGroup=false
    Non Uniform Scale=<1 1 -1>
    Override Seed=false
    Sphere Radius=&RandomRangeFloat(Minimum=10, Maximum=25, Evaluation Type=::NewEnumerator1)
    UseOwnerScale=true
   )
   &AddVelocity(Velocity=&Multiply_VectorByFloat(
     Vector=$Particles.ShapeLocation.ShapeVector
     Float=&RandomRangeFloat(Minimum=25, Maximum=80, Evaluation Type=::NewEnumerator1)
    )) {module=AddVelocity001}
   &CascadeConversion_InitialMeshRotation(Rotation:Vector3f=&RandomRangeVector(Minimum=<0 0 0>, Maximum=<1 1 1>, Evaluation Type=::NewEnumerator1))
  ]
  particleUpdateScript [
   &ParticleState()
   &ScaleMeshSize(Scale Factor=&VectorFromCurve(VectorCurve=DataInterface() {propertyValues {
       XCurve {keys [<at 0.2 1 RCIM_Cubic>, {arriveTangent=-1.25, interpMode=RCIM_Cubic, leaveTangent=-1.25, time=0.7932, value=0.7509}, <at 1 0>]}
       YCurve {keys [<at 0.2 1 RCIM_Cubic>, {arriveTangent=-0.25, interpMode=RCIM_Cubic, leaveTangent=-0.25, time=0.7932, value=0.85}, <at 1 0.8>]}
       ZCurve {keys [<at 0.2 1 RCIM_Cubic>,<at 0.7932 1 RCIM_Cubic>,<at 1 1>]}
      }}))
   &UpdateMeshOrientation(
    Rotation Rate=&FloatFromCurve(FloatCurve=DataInterface() {propertyValues {Curve {keys [{arriveTangent=-1.8976, interpMode=RCIM_Cubic, leaveTangent=-1.8976, tangentMode=RCTM_User, time=0, value=1}, <at 1 0 RCIM_Cubic>]}}})
    Rotation Vector=$Particles.SpawnRotationVector
   )
   &Drag(Drag=&RandomRangeFloat(Minimum=2, Maximum=3), Use Linear Drag=true, Use Rotational Drag=true)
   &AccelerationForce(Acceleration=<0 0 -200>)
   &SolveForcesAndVelocity(Clamp Velocity=false, Limit Acceleration=false)
   &ScaleColor(
    Scale Alpha=&FloatFromCurve(FloatCurve=DataInterface() {propertyValues {Curve {keys [<at 0.75 1 RCIM_Cubic>,<at 1 0>]}}})
    Scale RGB=&VectorFromFloat(Value=&FloatFromCurve(FloatCurve=DataInterface() {propertyValues {Curve {keys [<at 0.1 25 RCIM_Cubic>, {arriveTangent=-26.6667, interpMode=RCIM_Cubic, leaveTangent=-26.6667, time=0.1961, value=7.5192}, <at 1 1>]}}}))
    ScaleA=true
    ScaleRGB=true
   )
   &CameraOffset(Camera Offset Amount=-50)
  ]
  rendererClasses [&NiagaraMeshRendererProperties]
  renderers [{rendererClass=&NiagaraMeshRendererProperties, rendererIndex=0}]
 }
 Frags2 {
  emitterUpdateScript [
   &EmitterState(Life Cycle Mode=::NewEnumerator1, Loop Behavior=::NewEnumerator1, Loop Duration=0.5, UseLoopDelay=true)
   &SpawnBurst_Instantaneous(Spawn Count=5, Use Spawn Probability=false, UseLoopCountLimit=false)
  ]
  particleSpawnScript [
   /SetVariables(Particles.SpawnRotationVector=&RandomRangeVector(Minimum=<0.1 1.5 -0.1>, Maximum=<0.4 2 0.1>, Evaluation Type=::NewEnumerator1)) {isSetParameters=true}
   &InitializeParticle(
    Color=&MultiplyLinearColor(Linear Color=$User.FXColorC, Scale Factor=<10 10 10 1>)
    Lifetime=&RandomRangeFloat(Minimum=1.5, Maximum=2, Evaluation Type=::NewEnumerator1)
    Mesh Scale=&RandomRangeVector(Minimum=<2 2 3>, Maximum=<5 3 3>, Evaluation Type=::NewEnumerator1)
    Mesh Scale Mode=::NewEnumerator4
    Position Offset=<0 0 -50>
    UsePositionOffset=true
    Write Material Random=true
   )
   &AddVelocity(Velocity=&RandomRangeVector(Minimum=<0 0 -3500>, Maximum=<0 0 -1600>, Evaluation Type=::NewEnumerator1))
   &ShapeLocation(
    Hemisphere Distribution=<1 0.5>
    MaskSpawnGroup=false
    Non Uniform Scale=<1 1 -1>
    Override Seed=false
    Sphere Radius=&RandomRangeFloat(Minimum=10, Maximum=25, Evaluation Type=::NewEnumerator1)
    UseOwnerScale=true
   )
   &AddVelocity(Velocity=&Multiply_VectorByFloat(
     Vector=$Particles.ShapeLocation.ShapeVector
     Float=&RandomRangeFloat(Minimum=25, Maximum=80, Evaluation Type=::NewEnumerator1)
    )) {module=AddVelocity001}
   &CascadeConversion_InitialMeshRotation(Rotation:Vector3f=&RandomRangeVector(Minimum=<0 0 0>, Maximum=<1 1 1>, Evaluation Type=::NewEnumerator1))
  ]
  particleUpdateScript [
   &ParticleState()
   &ScaleMeshSize(Scale Factor=&VectorFromCurve(VectorCurve=DataInterface() {propertyValues {
       XCurve {keys [<at 0.2 1 RCIM_Cubic>, {arriveTangent=-1.25, interpMode=RCIM_Cubic, leaveTangent=-1.25, time=0.7932, value=0.7509}, <at 1 0>]}
       YCurve {keys [<at 0.2 1 RCIM_Cubic>, {arriveTangent=-0.25, interpMode=RCIM_Cubic, leaveTangent=-0.25, time=0.7932, value=0.85}, <at 1 0.8>]}
       ZCurve {keys [<at 0.2 1 RCIM_Cubic>,<at 0.7932 1 RCIM_Cubic>,<at 1 1>]}
      }}))
   &UpdateMeshOrientation(
    Rotation Rate=&FloatFromCurve(FloatCurve=DataInterface() {propertyValues {Curve {keys [
         {interpMode=RCIM_Cubic, leaveTangent=-1.5498, tangentMode=RCTM_Break, time=0, value=1}
         {interpMode=RCIM_Cubic, tangentMode=RCTM_Break, time=1, value=0}
        ]}}})
    Rotation Vector=$Particles.SpawnRotationVector
   )
   &Drag(Drag=&RandomRangeFloat(Minimum=2, Maximum=3), Use Linear Drag=true, Use Rotational Drag=true)
   &AccelerationForce(Acceleration=<0 0 -200>)
   &SolveForcesAndVelocity(Clamp Velocity=false, Limit Acceleration=false)
   &ScaleColor(
    Scale Alpha=&FloatFromCurve(FloatCurve=DataInterface() {propertyValues {Curve {keys [<at 0.75 1 RCIM_Cubic>,<at 1 0>]}}})
    Scale RGB=&VectorFromFloat(Value=&FloatFromCurve(FloatCurve=DataInterface() {propertyValues {Curve {keys [<at 0.1 25 RCIM_Cubic>, {arriveTangent=-26.6667, interpMode=RCIM_Cubic, leaveTangent=-26.6667, time=0.1961, value=7.5192}, <at 1 1>]}}}))
    ScaleA=true
    ScaleRGB=true
   )
   &CameraOffset(Camera Offset Amount=-50)
  ]
  rendererClasses [&NiagaraMeshRendererProperties]
  renderers [{rendererClass=&NiagaraMeshRendererProperties, rendererIndex=0}]
 }
 Frags3 {
  emitterUpdateScript [
   &EmitterState(Life Cycle Mode=::NewEnumerator1, Loop Behavior=::NewEnumerator1, Loop Duration=0.5, UseLoopDelay=true)
   &SpawnBurst_Instantaneous(Spawn Count=5, Use Spawn Probability=false, UseLoopCountLimit=false)
  ]
  particleSpawnScript [
   /SetVariables(Particles.SpawnRotationVector=&RandomRangeVector(Minimum=<0.1 1.5 -0.1>, Maximum=<0.4 2 0.1>, Evaluation Type=::NewEnumerator1)) {isSetParameters=true}
   &InitializeParticle(
    Color=&MultiplyLinearColor(Linear Color=$User.FXColorC, Scale Factor=<10 10 10 1>)
    Lifetime=&RandomRangeFloat(Minimum=1.5, Maximum=2, Evaluation Type=::NewEnumerator1)
    Mesh Scale=&RandomRangeVector(Minimum=<2 2 3>, Maximum=<5 3 3>, Evaluation Type=::NewEnumerator1)
    Mesh Scale Mode=::NewEnumerator4
    Position Offset=<0 0 -50>
    UsePositionOffset=true
    Write Material Random=true
   )
   &AddVelocity(Velocity=&RandomRangeVector(Minimum=<0 0 -3500>, Maximum=<0 0 -1600>, Evaluation Type=::NewEnumerator1))
   &ShapeLocation(
    Hemisphere Distribution=<1 0.5>
    MaskSpawnGroup=false
    Non Uniform Scale=<1 1 -1>
    Override Seed=false
    Sphere Radius=&RandomRangeFloat(Minimum=10, Maximum=25, Evaluation Type=::NewEnumerator1)
    UseOwnerScale=true
   )
   &AddVelocity(Velocity=&Multiply_VectorByFloat(
     Vector=$Particles.ShapeLocation.ShapeVector
     Float=&RandomRangeFloat(Minimum=25, Maximum=80, Evaluation Type=::NewEnumerator1)
    )) {module=AddVelocity001}
   &CascadeConversion_InitialMeshRotation(Rotation:Vector3f=&RandomRangeVector(Minimum=<0 0 0>, Maximum=<1 1 1>, Evaluation Type=::NewEnumerator1))
  ]
  particleUpdateScript [
   &ParticleState()
   &ScaleMeshSize(Scale Factor=&VectorFromCurve(VectorCurve=DataInterface() {propertyValues {
       XCurve {keys [<at 0.2 1 RCIM_Cubic>, {arriveTangent=-1.25, interpMode=RCIM_Cubic, leaveTangent=-1.25, time=0.7932, value=0.7509}, <at 1 0>]}
       YCurve {keys [<at 0.2 1 RCIM_Cubic>, {arriveTangent=-0.25, interpMode=RCIM_Cubic, leaveTangent=-0.25, time=0.7932, value=0.85}, <at 1 0.8>]}
       ZCurve {keys [<at 0.2 1 RCIM_Cubic>,<at 0.7932 1 RCIM_Cubic>,<at 1 1>]}
      }}))
   &UpdateMeshOrientation(
    Rotation Rate=&FloatFromCurve(FloatCurve=DataInterface() {propertyValues {Curve {keys [
         {interpMode=RCIM_Cubic, leaveTangent=-1.5424, tangentMode=RCTM_Break, time=0, value=1}
         {interpMode=RCIM_Cubic, tangentMode=RCTM_Break, time=1, value=0}
        ]}}})
    Rotation Vector=$Particles.SpawnRotationVector
   )
   &Drag(Drag=&RandomRangeFloat(Minimum=2, Maximum=3), Use Linear Drag=true, Use Rotational Drag=true)
   &AccelerationForce(Acceleration=<0 0 -200>)
   &SolveForcesAndVelocity(Clamp Velocity=false, Limit Acceleration=false)
   &ScaleColor(
    Scale Alpha=&FloatFromCurve(FloatCurve=DataInterface() {propertyValues {Curve {keys [<at 0.75 1 RCIM_Cubic>,<at 1 0>]}}})
    Scale RGB=&VectorFromFloat(Value=&FloatFromCurve(FloatCurve=DataInterface() {propertyValues {Curve {keys [<at 0.1 25 RCIM_Cubic>, {arriveTangent=-26.6667, interpMode=RCIM_Cubic, leaveTangent=-26.6667, time=0.1961, value=7.5192}, <at 1 1>]}}}))
    ScaleA=true
    ScaleRGB=true
   )
   &CameraOffset(Camera Offset Amount=-50)
  ]
  rendererClasses [&NiagaraMeshRendererProperties]
  renderers [{rendererClass=&NiagaraMeshRendererProperties, rendererIndex=0}]
 }
 RoundFlare {
  emitterUpdateScript [
   &EmitterState(Inactive Response=::NewEnumerator1, Life Cycle Mode=::NewEnumerator1, Loop Behavior=::NewEnumerator1, Loop Duration=1, UseLoopDelay=true)
   &SpawnBurst_Instantaneous(Use Spawn Probability=false, UseLoopCountLimit=false)
  ]
  particleSpawnScript [&InitializeParticle(
    Color=&MultiplyLinearColor(Linear Color=$User.FXColorB, Scale Factor=<3 3 3 1>)
    Lifetime=0.4
    Position Offset=<0 0 -125>
    Sprite Size=<1200 1200>
    Sprite Size Mode=::NewEnumerator4
    UsePositionOffset=true
    Write Material Random=true
   )]
  particleUpdateScript [
   &ParticleState()
   &ScaleSpriteSize(
    Scale Sprite Size Mode=::NewEnumerator0
    Uniform Scale Factor=&FloatFromCurve(FloatCurve=DataInterface() {propertyValues {Curve {keys [
         {arriveTangent=16.2971, interpMode=RCIM_Cubic, leaveTangent=16.2971, tangentMode=RCTM_User, time=0, value=0}
         {arriveTangentWeight=0.0479, interpMode=RCIM_Cubic, leaveTangentWeight=0.1858, tangentMode=RCTM_User, time=0.05, value=1}
         {arriveTangent=-0.4374, interpMode=RCIM_Cubic, leaveTangent=-0.4374, tangentMode=RCTM_User, time=0.2439, value=0.6576}
         {arriveTangent=-0.8672, interpMode=RCIM_Cubic, leaveTangent=-0.8672, tangentMode=RCTM_User, time=1, value=0}
        ]}}})
   )
   &ScaleColor(
    Scale Alpha=&FloatFromCurve(FloatCurve=DataInterface() {propertyValues {Curve {keys [<at 0 1 RCIM_Cubic>, {arriveTangent=-1.9891, interpMode=RCIM_Cubic, leaveTangent=-1.9891, tangentMode=RCTM_User, time=1, value=0}]}}})
    Scale RGB=&VectorFromFloat(Value=&FloatFromCurve(FloatCurve=DataInterface() {propertyValues {Curve {keys [{arriveTangent=-8.475, interpMode=RCIM_Cubic, leaveTangent=-8.475, tangentMode=RCTM_User, time=0, value=5}, <at 1 1 RCIM_Cubic>]}}}))
    ScaleA=true
    ScaleRGB=true
   )
   &DynamicMaterialParameters(
    Index 0 Param 1=&FloatFromCurve(FloatCurve=DataInterface() {propertyValues {Curve {keys [{arriveTangent=-3.8495, interpMode=RCIM_Cubic, time=0, value=0.9}, <at 0.5 0.05>]}}})
    Index 0 Param 2=2
    Index 0 Param 3=1
    Index 0 Param 4=1
    Module.Index0Param0Write=true
    Module.Index0Param1Write=true
    Module.Index0Param2Write=true
    Module.Index0Param3Write=true
   )
   &CameraOffset(Camera Offset Amount=300)
  ]
  rendererClasses [&NiagaraSpriteRendererProperties]
  renderers [{rendererClass=&NiagaraSpriteRendererProperties}]
 }
 ShockWave {
  emitterUpdateScript [
   &EmitterState(Life Cycle Mode=::NewEnumerator1, Loop Behavior=::NewEnumerator1, Loop Duration=1, UseLoopDelay=true)
   &SpawnBurst_Instantaneous(Use Spawn Probability=false, UseLoopCountLimit=false)
  ]
  particleSpawnScript [&InitializeParticle(Color=<rgba 1 1 1 0.1>, Lifetime=0.6, Mesh Scale Mode=::NewEnumerator4, UsePositionOffset=false, Write Material Random=true)]
  particleUpdateScript [
   &ParticleState()
   &ScaleMeshSize(Scale Factor=&VectorFromCurve(VectorCurve=DataInterface() {propertyValues {XCurve {keys [<at 0 0.2>,<at 1 8>]}, YCurve {keys [<at 0 0.2>,<at 1 8>]}, ZCurve {keys [<at 0 0.2>,<at 1 8>]}}}))
   &ScaleColor(
    Scale Alpha=&FloatFromCurve(FloatCurve=DataInterface() {propertyValues {Curve {keys [{arriveTangent=-1.5832, interpMode=RCIM_Cubic, leaveTangent=-1.5832, tangentMode=RCTM_User, time=0, value=1}, <at 1 0 RCIM_Cubic>]}}})
    ScaleA=true
    ScaleRGB=true
   )
   &CameraOffset(Camera Offset Amount=500)
  ]
  rendererClasses [&NiagaraMeshRendererProperties]
  renderers [{rendererClass=&NiagaraMeshRendererProperties, rendererIndex=0}]
 }
 Sparklies {
  simTarget=GPUComputeSim
  emitterUpdateScript [
   &EmitterState(Life Cycle Mode=::NewEnumerator1, Loop Behavior=::NewEnumerator1, Loop Duration=0.25, UseLoopDelay=true)
   &SpawnBurst_Instantaneous(Spawn Count=15, Spawn Time=0.04, Use Spawn Probability=false, UseLoopCountLimit=false) {module=SpawnBurst_Instantaneous001}
   &SpawnBurst_Instantaneous(Spawn Count=20, Use Spawn Probability=false, UseLoopCountLimit=false)
   &SpawnRate(SpawnRate=300, Use Spawn Probability=false)
  ]
  particleSpawnScript [
   /SetVariables(Particles.SpriteRotationRate=&RandomRangeFloat(Minimum=-1, Maximum=1, Evaluation Type=::NewEnumerator1)) {isSetParameters=true}
   &InitializeParticle(
    Color=&MultiplyLinearColor(
     Linear Color=$User.FXColorB
     Scale Factor=&MakeFloat4FromVectorAndFloat("Vector (RGB)"=&VectorFromFloat(Value=2), "Float (Alpha)"=1)
    )
    Lifetime=&RandomRangeFloat(Minimum=0.25, Maximum=1.25, Evaluation Type=::NewEnumerator1)
    Position Offset=<0 0 -50>
    Sprite Rotation Angle=&RandomRangeFloat(Minimum=0, Maximum=1, Evaluation Type=::NewEnumerator1)
    Sprite Rotation Mode=::NewEnumerator1
    Sprite Size=&Vector2DFromFloat(Value=&RandomRangeFloat(Minimum=50, Maximum=80, Evaluation Type=::NewEnumerator1))
    Sprite Size Mode=::NewEnumerator4
    UsePositionOffset=true
    Write Material Random=true
   )
   &AddVelocity(Velocity=&RandomRangeVector(Minimum=<0 0 -5000>, Maximum=<0 0 -1500>, Evaluation Type=::NewEnumerator1))
   &ShapeLocation(
    Hemisphere Distribution=<1 0.5>
    MaskSpawnGroup=false
    Non Uniform Scale=<1 1 -1>
    Override Seed=false
    Sphere Radius=&RandomRangeFloat(Minimum=5, Maximum=28, Evaluation Type=::NewEnumerator1)
    UseOwnerScale=true
   )
   &AddVelocity(Velocity=&Multiply_VectorByFloat(
     Vector=$Particles.ShapeLocation.ShapeVector
     Float=&RandomRangeFloat(Minimum=0, Maximum=50, Evaluation Type=::NewEnumerator1)
    )) {module=AddVelocity001}
  ]
  particleUpdateScript [
   &ParticleState()
   &SpriteRotationRate(Rotation Rate=&AngleConversion(
     Angle:NiagaraFloat=$Particles.SpriteRotationRate
     Angle Input:ENiagara_AngleInput=::NewEnumerator1
     Angle Output:ENiagara_AngleInput=::NewEnumerator0
    ))
   &ScaleSpriteSize(
    Scale Sprite Size Mode=::NewEnumerator0
    Uniform Scale Factor=&FloatFromCurve(FloatCurve=DataInterface() {propertyValues {Curve {keys [
         {arriveTangent=-0.4739, time=-0.0044, value=1}
         {arriveTangent=0.2957, interpMode=RCIM_Cubic, leaveTangent=0.2957, time=0.157, value=0.5261}
         {arriveTangent=1.0007, interpMode=RCIM_Cubic, leaveTangent=1.0007, time=0.1686, value=1.0512}
         {arriveTangent=-3.1557, interpMode=RCIM_Cubic, leaveTangent=-3.1557, time=0.1876, value=0.5567}
         {arriveTangentWeight=0.107, interpMode=RCIM_Cubic, leaveTangentWeight=0.0142, tangentMode=RCTM_User, time=0.3841, value=0.3711}
         {arriveTangent=4.996, interpMode=RCIM_Cubic, leaveTangent=4.996, time=0.4102, value=0.8441}
         {arriveTangent=-1.6571, interpMode=RCIM_Cubic, leaveTangent=-1.6571, time=0.4265, value=0.583}
         {arriveTangent=0.0755, interpMode=RCIM_Cubic, leaveTangent=0.0755, time=0.69, value=0.3804}
         {arriveTangent=0.7225, interpMode=RCIM_Cubic, leaveTangent=0.7225, time=0.7007, value=0.6037}
         {arriveTangent=-2.017, interpMode=RCIM_Cubic, leaveTangent=-2.017, time=0.7133, value=0.3972}
         <at 1 0>
        ]}}})
   )
   &Drag(Drag=&RandomRangeFloat(Minimum=2, Maximum=4), Use Linear Drag=true, Use Rotational Drag=true)
   &SolveForcesAndVelocity(Clamp Velocity=false, Limit Acceleration=false)
   &ScaleColor(
    Scale RGB=&VectorFromFloat(Value=&FloatFromCurve(FloatCurve=DataInterface() {propertyValues {Curve {keys [<at 0 2 RCIM_Cubic>,<at 0.4 1>]}}}))
    ScaleA=true
    ScaleRGB=true
   )
  ]
  rendererClasses [&NiagaraSpriteRendererProperties]
  renderers [{rendererClass=&NiagaraSpriteRendererProperties}]
 }
}
summary {
 systemName=NS_Jolly_Firework_Rift_Shatter
 emitters [
  {emitterName=ElectricArcsResidual, rendererClasses [&NiagaraMeshRendererProperties]}
  {emitterName=Endflare, rendererClasses [&NiagaraSpriteRendererProperties]}
  {emitterName=ShockWave, rendererClasses [&NiagaraMeshRendererProperties]}
  {emitterName=Sparklies, simTarget=GPUComputeSim, rendererClasses [&NiagaraSpriteRendererProperties]}
  {emitterName=Frags1, rendererClasses [&NiagaraMeshRendererProperties]}
  {emitterName=Frags2, rendererClasses [&NiagaraMeshRendererProperties]}
  {emitterName=Frags3, rendererClasses [&NiagaraMeshRendererProperties]}
  {emitterName=RoundFlare, rendererClasses [&NiagaraSpriteRendererProperties]}
 ]
 userVariables [
  User.FXColorA {type=LinearColor, description="", defaultValue {struct=LinearColor, value=<rgba 0.3533 0.03 1 1>}}
  User.FXColorB {type=LinearColor, description="", defaultValue {struct=LinearColor, value=<rgba 0.12 0.8533 1 1>}}
  User.FXColorC {type=LinearColor, description="", defaultValue {struct=LinearColor, value=<rgba 1 1 1 1>}}
 ]
}
