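# Niagara trail effect for the Jolly Rascal consumable's projectile, styled as a burning fuse with a hot pink-magenta glow.
# Seven independent emitters layer short-lived core, fuse and gradient glow sprites, a flickering purple light, and a ribbon smoke trail pushed along negative X that cools to dark grey.
# GPU-simulated sparks and streaks fall under gravity with drag and curve-driven size, with the sparks also stretched by speed and color-shifted over life.
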NiagaraSystem /Game/Athena/Items/Consumables/JollyRascal/FX/NS_Jolly_Prj_Trail
dependencies {
 usedDataInterfaces [
  &NiagaraDataInterfaceCurve
  &NiagaraDataInterfaceVectorCurve
  &NiagaraDataInterfaceVector2DCurve
 ]
 usedDynamicInputs []
 usedModules [&SystemState]
}
emitters {
 Core {
  emitterUpdateScript [
   &EmitterState(Inactive Response=::NewEnumerator1, Life Cycle Mode=::NewEnumerator1, Loop Duration=0.5, UseLoopDelay=true)
   &SpawnRate(SpawnRate=40, Use Spawn Probability=false)
  ]
  particleSpawnScript [
   &InitializeParticle(
    Color=<rgba 70 9.0633 30.4 1>
    Lifetime=&RandomRangeFloat(Minimum=0.05, Maximum=0.08, Evaluation Type=::NewEnumerator1)
    Sprite Rotation Angle=&RandomRangeFloat(Minimum=0, Maximum=1, Evaluation Type=::NewEnumerator1)
    Sprite Rotation Mode=::NewEnumerator1
    Sprite Size=&Vector2DFromFloat(Value=&RandomRangeFloat(Minimum=11, Maximum=16, Evaluation Type=::NewEnumerator1))
    Sprite Size Mode=::NewEnumerator4
    UsePositionOffset=false
    Write Material Random=true
   )
   &CameraOffset(Camera Offset Amount=&RandomRangeFloat(Minimum=-2, Maximum=5, Evaluation Type=::NewEnumerator1))
  ]
  particleUpdateScript [
   &ParticleState()
   &ScaleSpriteSize(
    Scale Sprite Size Mode=::NewEnumerator0
    Uniform Scale Factor=&FloatFromCurve(FloatCurve=DataInterface() {propertyValues {Curve {keys [
         {arriveTangent=1.5937, interpMode=RCIM_Cubic, leaveTangent=1.5937, leaveTangentWeight=0.439, tangentMode=RCTM_User, time=0, value=0}
         {arriveTangentWeight=0.2609, interpMode=RCIM_Cubic, leaveTangentWeight=0.1118, tangentMode=RCTM_User, time=0.7, value=1}
         {arriveTangent=-1.6833, arriveTangentWeight=0.1, interpMode=RCIM_Cubic, leaveTangent=-1.6833, tangentMode=RCTM_User, time=1, value=0.5}
        ]}}})
   )
   /Niagara/Modules/Update/SubUV/SubUVAnimation(Number Of Frames=3)
  ]
  rendererClasses [&NiagaraSpriteRendererProperties]
  renderers [{rendererClass=&NiagaraSpriteRendererProperties}]
 }
 FireGradient {
  emitterUpdateScript [
   &EmitterState(Inactive Response=::NewEnumerator1, Life Cycle Mode=::NewEnumerator1, Loop Duration=1, UseLoopDelay=true)
   &SpawnRate(SpawnRate=50, Use Spawn Probability=false)
  ]
  particleSpawnScript [
   &InitializeParticle(
    Color=<rgba 5 5 5 0.1>
    Lifetime=&RandomRangeFloat(Minimum=0.1, Maximum=0.2, Evaluation Type=::NewEnumerator1)
    Sprite Size=<50 50>
    Sprite Size Mode=::NewEnumerator4
    UsePositionOffset=false
    Write Material Random=true
   )
   &Color(Color=&MakeLinearColorFromVectorAndFloat("Vector (RGB)"=<5 5 5>, "Float (Alpha)"=&ClampFloat(Float=0.1))) {enabled=false}
  ]
  particleUpdateScript [
   &ParticleState()
   &ScaleSpriteSize(Scale Sprite Size Mode=::NewEnumerator0)
   &ScaleColor(
    Scale Alpha=&FloatFromCurve(FloatCurve=DataInterface() {propertyValues {Curve {keys [<at 0 0 RCIM_Cubic>,<at 0.5 1 RCIM_Cubic>,<at 1 0 RCIM_Cubic>]}}})
    Scale RGB=&VectorFromCurve(VectorCurve=DataInterface() {propertyValues {
       XCurve {keys [
         <at 0 0.1146 RCIM_Cubic>
         {arriveTangent=3.3899, interpMode=RCIM_Cubic, leaveTangent=-3.9883, tangentMode=RCTM_Break, time=0.5, value=1}
         <at 1 0.1146 RCIM_Cubic>
        ]}
       YCurve {keys [<at 0 0.0519 RCIM_Cubic>,<at 0.5 0.0484 RCIM_Cubic>,<at 1 0.0519 RCIM_Cubic>]}
       ZCurve {keys [
         <at 0 0.0288 RCIM_Cubic>
         {arriveTangent=2.0106, interpMode=RCIM_Cubic, leaveTangent=-2.6555, tangentMode=RCTM_Break, time=0.5, value=0.7128}
         <at 1 0.0288 RCIM_Cubic>
        ]}
      }})
    ScaleA=true
    ScaleRGB=true
   )
   &DynamicMaterialParameters(
    Index 0 Param 1=&FloatFromCurve(FloatCurve=DataInterface() {propertyValues {Curve {keys [<at 0 8>,<at 0.5 5>,<at 1 8>]}}})
    Index 0 Param 2=1
    Index 0 Param 3=1
    Index 0 Param 4=1
    Module.Index0Param0Write=true
    Module.Index0Param1Write=true
    Module.Index0Param2Write=true
    Module.Index0Param3Write=true
   )
   &CameraOffset(Camera Offset Amount=50)
  ]
  rendererClasses [&NiagaraSpriteRendererProperties]
  renderers [{rendererClass=&NiagaraSpriteRendererProperties}]
 }
 Light {
  emitterUpdateScript [
   &EmitterState(
    Inactive Response=::NewEnumerator1
    Life Cycle Mode=::NewEnumerator1
    Loop Duration=&RandomRangeFloat(Minimum=0.15, Maximum=0.2, Evaluation Type=::NewEnumerator1)
    Recalculate Duration Each Loop=true
    UseLoopDelay=true
   )
   &SpawnBurst_Instantaneous(Use Spawn Probability=false, UseLoopCountLimit=false)
  ]
  particleSpawnScript [
   &InitializeParticle(
    Lifetime=&RandomRangeFloat(Minimum=0.1, Maximum=0.2, Evaluation Type=::NewEnumerator1)
    Sprite Rotation Angle=&RandomRangeFloat(Minimum=0, Maximum=1, Evaluation Type=::NewEnumerator1)
    Sprite Rotation Mode=::NewEnumerator1
    Sprite Size=&Vector2DFromFloat(Value=&RandomRangeFloat(Minimum=500, Maximum=700, Evaluation Type=::NewEnumerator1))
    Sprite Size Mode=::NewEnumerator4
    UsePositionOffset=false
    Write Material Random=true
   )
   &Color(Color=&MakeLinearColorFromVectorAndFloat("Vector (RGB)"=<0.831 0.0743 1>, "Float (Alpha)"=&ClampFloat(Float=1)))
  ]
  particleUpdateScript [
   /SetVariables(Particles.LightExponent=14) {isSetParameters=true}
   &ParticleState()
   &ScaleColor(
    Scale Alpha=&FloatFromCurve(FloatCurve=DataInterface() {propertyValues {Curve {keys [
         {arriveTangent=4.3581, interpMode=RCIM_Cubic, time=0, value=0}
         <at 0.5 1 RCIM_Cubic>
         {interpMode=RCIM_Cubic, leaveTangent=-3.9938, time=1, value=0}
        ]}}})
    ScaleA=true
    ScaleRGB=true
   )
   &CascadeConversion_LightProperties(LightBrightnessScale=20, LightColorScale=<1 1 1>, LightRadiusScale=1)
  ]
  rendererClasses [&NiagaraLightRendererProperties]
  renderers [{rendererClass=&NiagaraLightRendererProperties, rendererIndex=0}]
 }
 SmokeTrail {
  emitterUpdateScript [
   &EmitterState(Life Cycle Mode=::NewEnumerator1, Loop Duration=1, UseLoopDelay=true)
   &SpawnRate(SpawnRate=20, Use Spawn Probability=false)
  ]
  particleSpawnScript [
   &InitializeRibbon(Lifetime=0.5, Ribbon Width=&Multiply_Float(A=12, B=2), UsePositionOffset=false, Write Material Random=true)
   &Color(Color=&MakeLinearColorFromVectorAndFloat("Vector (RGB)"=<1 1 1>, "Float (Alpha)"=&ClampFloat(Float=1))) {enabled=false}
   &AddVelocity(Velocity=<-500 0 0>)
  ]
  particleUpdateScript [
   /SetVariables(Particles.ScaledRibbonWidth:NiagaraFloat=&Multiply_Float(
     A=$Particles.RibbonWidth
     B=&FloatFromCurve(FloatCurve=DataInterface() {propertyValues {Curve {keys [<at 0 1 RCIM_Cubic>, {arriveTangent=0.5, interpMode=RCIM_Cubic, leaveTangent=0.5, time=0.1606, value=0.7431}, <at 1 1.5>]}}})
    )) {isSetParameters=true}
   &ParticleState()
   &ScaleColor(
    Scale RGB=&VectorFromCurve(VectorCurve=DataInterface() {propertyValues {
       XCurve {keys [
         {arriveTangent=-12.9886, interpMode=RCIM_Cubic, leaveTangent=-12.9886, tangentMode=RCTM_User, time=0.05, value=10}
         {arriveTangent=-11.3161, interpMode=RCIM_Cubic, leaveTangent=-11.3161, tangentMode=RCTM_User, time=0.15, value=1.3935}
         {arriveTangentWeight=0.1517, interpMode=RCIM_Cubic, leaveTangentWeight=0.3642, tangentMode=RCTM_User, time=0.4, value=0.1167}
         <at 1 0.1>
        ]}
       YCurve {keys [
         <at 0.05 0.5065 RCIM_Cubic>
         {arriveTangent=-1.1295, interpMode=RCIM_Cubic, leaveTangent=-1.1295, tangentMode=RCTM_User, time=0.15, value=0.4455}
         {arriveTangent=-0.0258, interpMode=RCIM_Cubic, leaveTangent=-0.0258, tangentMode=RCTM_User, time=0.4, value=0.1179}
         <at 1 0.1>
        ]}
       ZCurve {keys [
         <at 0.05 1.0041 RCIM_Cubic>
         {arriveTangentWeight=0.0855, interpMode=RCIM_Cubic, leaveTangentWeight=0.2138, tangentMode=RCTM_User, time=0.15, value=3}
         {arriveTangent=-0.0783, interpMode=RCIM_Cubic, leaveTangent=-0.0783, tangentMode=RCTM_User, time=0.4, value=0.1771}
         <at 1 0.1>
        ]}
      }})
    ScaleA=true
    ScaleRGB=true
   )
   &Drag(Drag=2, Use Linear Drag=true, Use Rotational Drag=true)
   &SolveForcesAndVelocity(Clamp Velocity=false, Limit Acceleration=false)
   &DynamicMaterialParameters(
    Index 0 Param 1=&FloatFromCurve(FloatCurve=DataInterface() {propertyValues {Curve {keys [
         {arriveTangent=13.1928, interpMode=RCIM_Cubic, leaveTangent=13.1928, tangentMode=RCTM_User, time=0, value=0}
         {arriveTangentWeight=0.1503, interpMode=RCIM_Cubic, leaveTangentWeight=0.1107, tangentMode=RCTM_User, time=0.1, value=1}
         {arriveTangent=-0.5099, interpMode=RCIM_Cubic, leaveTangent=-0.5099, tangentMode=RCTM_User, time=0.1736, value=0.7636}
         {arriveTangent=-0.924, interpMode=RCIM_Cubic, leaveTangent=-0.924, time=0.8596, value=0.4139}
         {arriveTangent=-1.3582, interpMode=RCIM_Cubic, leaveTangent=-1.3582, tangentMode=RCTM_User, time=1, value=0}
        ]}}})
    Index 0 Param 2=&FloatFromCurve(FloatCurve=DataInterface() {propertyValues {Curve {keys [<at 0 1 RCIM_Cubic>,<at 0.5 0>]}}})
    Index 0 Param 3=-0.5
    Index 0 Param 4=1
    Module.Index0Param0Write=true
    Module.Index0Param1Write=true
    Module.Index0Param2Write=true
    Module.Index0Param3Write=true
   )
  ]
  rendererClasses [&NiagaraRibbonRendererProperties]
  renderers [{rendererClass=&NiagaraRibbonRendererProperties}]
 }
 SparksClose {
  simTarget=GPUComputeSim
  emitterUpdateScript [
   &EmitterState(Life Cycle Mode=::NewEnumerator1, Loop Duration=0.5, UseLoopDelay=true)
   &SpawnRate(SpawnRate=400, Use Spawn Probability=false)
  ]
  particleSpawnScript [
   &InitializeParticle(
    Color=<rgba 25 25 25 1>
    Lifetime=&RandomRangeFloat(Minimum=0.1, Maximum=0.7, Evaluation Type=::NewEnumerator1)
    Sprite Size=&RandomRangeVector2D(Minimum=<1.5 1.5>, Maximum=<3 3>, Evaluation Type=::NewEnumerator1)
    Sprite Size Mode=::NewEnumerator4
    UsePositionOffset=false
    Write Material Random=true
   )
   &Color(Color=&MakeLinearColorFromVectorAndFloat("Vector (RGB)"=<25 25 25>, "Float (Alpha)"=&ClampFloat(Float=1))) {enabled=false}
   &AddVelocity(Velocity=&RandomRangeVector(Minimum=<-300 0 0>, Maximum=<-150 0 0>, Evaluation Type=::NewEnumerator1))
   &ShapeLocation(MaskSpawnGroup=false, Override Seed=false, Sphere Radius=3, UseOwnerScale=true)
   &AddVelocity(Velocity=&Multiply_VectorByFloat(
     Vector=$Particles.ShapeLocation.ShapeVector
     Float=&RandomRangeFloat(Minimum=25, Maximum=120, Evaluation Type=::NewEnumerator1)
    )) {module=AddVelocity001}
  ]
  particleUpdateScript [
   &ParticleState()
   &ScaleSpriteSize(
    Scale Factor=&Vector2DFromCurve(
     Vector2Curve=DataInterface() {propertyValues {
       XCurve {keys [
         {arriveTangent=5.939, interpMode=RCIM_Cubic, time=0, value=0.5}
         {arriveTangent=-0.4, interpMode=RCIM_Cubic, leaveTangent=-0.4, time=0.15, value=1}
         <at 1 0.1>
        ]}
       YCurve {keys [
         {arriveTangent=28.4834, interpMode=RCIM_Cubic, time=0, value=0}
         {arriveTangent=0.15, interpMode=RCIM_Cubic, leaveTangent=0.15, time=0.15, value=1.5}
         <at 1 0.15>
        ]}
      }}
     ScaleCurve=<1 1>
    )
    Scale Sprite Size Mode=::NewEnumerator2
   )
   &AccelerationForce(Acceleration=<0 0 -1600>)
   &Drag(Drag=&RandomRangeFloat(Minimum=0, Maximum=0.4), Use Linear Drag=true, Use Rotational Drag=true)
   &SolveForcesAndVelocity(Clamp Velocity=false, Limit Acceleration=false)
   &ScaleSpriteSizeBySpeed(
    Max Scale Factor=<1 2>
    Source Velocity:Vector3f=&Multiply_Vector(A=<0.05 0.02 1>, B=$Particles.Velocity)
    Velocity Threshold=300
   )
   &ScaleColor(
    Scale Alpha=&FloatFromCurve(FloatCurve=DataInterface() {propertyValues {Curve {keys [<at 0.7 1 RCIM_Cubic>,<at 1 0>]}}})
    Scale RGB=&VectorFromCurve(VectorCurve=DataInterface() {propertyValues {
       XCurve {keys [<at 0 10 RCIM_Cubic>, {arriveTangent=-9.3161, interpMode=RCIM_Cubic, leaveTangent=-9.3161, time=0.16, value=2}, <at 1 0.6839 RCIM_Cubic>]}
       YCurve {keys [
         <at 0 1.6521 RCIM_Cubic>
         {arriveTangent=-1.6043, interpMode=RCIM_Cubic, leaveTangent=-1.6043, time=0.16, value=0.1599}
         <at 1 0.0478 RCIM_Cubic>
        ]}
       ZCurve {keys [<at 0 0.5552 RCIM_Cubic>, {arriveTangent=0.4448, interpMode=RCIM_Cubic, leaveTangent=0.4448, time=0.16, value=1.0993}, <at 1 1 RCIM_Cubic>]}
      }})
    ScaleA=true
    ScaleRGB=true
   )
  ]
  rendererClasses [&NiagaraSpriteRendererProperties]
  renderers [{rendererClass=&NiagaraSpriteRendererProperties}]
 }
 Streaks {
  simTarget=GPUComputeSim
  emitterUpdateScript [
   &EmitterState(Life Cycle Mode=::NewEnumerator1, Loop Duration=1, UseLoopDelay=true)
   &SpawnRate(SpawnRate=200, Use Spawn Probability=false)
  ]
  particleSpawnScript [
   &InitializeParticle(
    Color=<rgba 20 4.6063 1.6076 1>
    Lifetime=&RandomRangeFloat(Minimum=0.025, Maximum=0.05, Evaluation Type=::NewEnumerator1)
    Sprite Size=&RandomRangeVector2D(Minimum=<1.5 4>, Maximum=<2 7>, Evaluation Type=::NewEnumerator1)
    Sprite Size Mode=::NewEnumerator4
    UsePositionOffset=false
    Write Material Random=true
   )
   &AddVelocity(Velocity=&RandomRangeVector(Minimum=<-250 -250 -250>, Maximum=<250 250 250>, Evaluation Type=::NewEnumerator1))
   &Color(Color=&MakeLinearColorFromVectorAndFloat("Vector (RGB)"=<20 4.6063 1.6076>, "Float (Alpha)"=&ClampFloat(Float=1))) {enabled=false}
  ]
  particleUpdateScript [
   &ParticleState()
   &ScaleSpriteSize(
    Scale Factor=&Vector2DFromCurve(
     Vector2Curve=DataInterface() {propertyValues {
       XCurve {keys [<at 0 1 RCIM_Cubic>, {arriveTangent=-1.0412, interpMode=RCIM_Cubic, leaveTangent=-1.0412, tangentMode=RCTM_User, time=1, value=0.2}]}
       YCurve {keys [<at 0 3 RCIM_Cubic>, {interpMode=RCIM_Cubic, leaveTangent=-4.2938, time=1, value=0}]}
      }}
     ScaleCurve=<1 1>
    )
    Scale Sprite Size Mode=::NewEnumerator2
   )
   &AccelerationForce(Acceleration=<0 0 -980>)
   &Drag(Use Linear Drag=true, Use Rotational Drag=true)
   &SolveForcesAndVelocity(Clamp Velocity=false, Limit Acceleration=false)
  ]
  rendererClasses [&NiagaraSpriteRendererProperties]
  renderers [{rendererClass=&NiagaraSpriteRendererProperties}]
 }
 fuse {
  emitterUpdateScript [
   &EmitterState(Life Cycle Mode=::NewEnumerator1, Loop Duration=0.5, UseLoopDelay=true)
   &SpawnBurst_Instantaneous(Use Spawn Probability=false, UseLoopCountLimit=false)
   &SpawnRate(SpawnRate=30, Use Spawn Probability=false)
  ]
  particleSpawnScript [
   /SetVariables(Particles.SpriteRotationRate=&RandomRangeFloat(Minimum=0, Maximum=0.01, Evaluation Type=::NewEnumerator1)) {isSetParameters=true}
   &InitializeParticle(
    Lifetime=&RandomRangeFloat(Minimum=0.05, Maximum=0.1, Evaluation Type=::NewEnumerator1)
    Sprite Rotation Angle=&RandomRangeFloat(Minimum=0, Maximum=1, Evaluation Type=::NewEnumerator1)
    Sprite Rotation Mode=::NewEnumerator1
    Sprite Size=&RandomRangeVector2D(Minimum=<40 5>, Maximum=<80 60>, Evaluation Type=::NewEnumerator1)
    Sprite Size Mode=::NewEnumerator4
    UsePositionOffset=false
    Write Material Random=true
   )
   &CameraOffset(Camera Offset Amount=&RandomRangeFloat(Minimum=-2, Maximum=5, Evaluation Type=::NewEnumerator1))
  ]
  particleUpdateScript [
   &ParticleState()
   &SpriteRotationRate(Rotation Rate=&AngleConversion(
     Angle:NiagaraFloat=$Particles.SpriteRotationRate
     Angle Input:ENiagara_AngleInput=::NewEnumerator1
     Angle Output:ENiagara_AngleInput=::NewEnumerator0
    ))
   &ScaleSpriteSize(
    Scale Factor=&Vector2DFromFloat(Value=&FloatFromCurve(FloatCurve=DataInterface() {propertyValues {Curve {keys [
          <at -0.0011 0.0573>
          {arriveTangent=-0.0322, interpMode=RCIM_Cubic, leaveTangent=-0.0322, time=0.0867, value=0.7609}
          {interpMode=RCIM_Cubic, leaveTangent=0.0935, time=1.0043, value=0.0249}
         ]}}}))
    Scale Sprite Size Mode=::NewEnumerator2
   )
   &Color(Color=&MakeLinearColorFromVectorAndFloat(
     "Vector (RGB)"=<100 9.4229 55.3134>
     "Float (Alpha)"=&FloatFromCurve(FloatCurve=DataInterface() {propertyValues {Curve {keys [<at 0 0 RCIM_Cubic>,<at 0.0967 0.1386 RCIM_Cubic>,<at 1 0>]}}})
    ))
  ]
  rendererClasses [&NiagaraSpriteRendererProperties]
  renderers [{rendererClass=&NiagaraSpriteRendererProperties}]
 }
}
summary {
 systemName=NS_Jolly_Prj_Trail
 emitters [
  {emitterName=FireGradient, rendererClasses [&NiagaraSpriteRendererProperties]}
  {emitterName=Streaks, simTarget=GPUComputeSim, rendererClasses [&NiagaraSpriteRendererProperties]}
  {emitterName=Light, rendererClasses [&NiagaraLightRendererProperties]}
  {emitterName=Core, rendererClasses [&NiagaraSpriteRendererProperties]}
  {emitterName=SmokeTrail, rendererClasses [&NiagaraRibbonRendererProperties]}
  {emitterName=fuse, rendererClasses [&NiagaraSpriteRendererProperties]}
  {emitterName=SparksClose, simTarget=GPUComputeSim, rendererClasses [&NiagaraSpriteRendererProperties]}
 ]
 userVariables []
}
