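# Niagara system for the Jolly Rascal consumable that draws a smoke trail, apparently behind small falling objects.
# A single SmokeTrail emitter spawns short-lived, translucent ribbon particles at a steady rate.
# Over each particle's life, curves dip then widen the ribbon and shift its color from bright cyan-blue to dim gray.
# Curve-driven dynamic material parameters animate the ribbon material, while drag damps particle motion.
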NiagaraSystem /Game/Athena/Items/Consumables/JollyRascal/FX/NS_Jolly_SmallObject_Falling
dependencies {
 usedDataInterfaces [&NiagaraDataInterfaceCurve, &NiagaraDataInterfaceVectorCurve]
 usedDynamicInputs []
 usedModules [&SystemState]
}
emitters {SmokeTrail {
  emitterUpdateScript [
   &EmitterState(Life Cycle Mode=::NewEnumerator1, Loop Duration=1, UseLoopDelay=true)
   &SpawnRate(SpawnRate=50, Use Spawn Probability=false)
  ]
  particleSpawnScript [
   &InitializeRibbon(Lifetime=0.25, Ribbon Width=&Multiply_Float(A=65, B=2), UsePositionOffset=false, Write Material Random=true)
   &Color(Color=&MakeLinearColorFromVectorAndFloat("Vector (RGB)"=<1 1 1>, "Float (Alpha)"=&ClampFloat(Float=0.25)))
  ]
  particleUpdateScript [
   /SetVariables(Particles.ScaledRibbonWidth:NiagaraFloat=&Multiply_Float(
     A=$Particles.RibbonWidth
     B=&FloatFromCurve(FloatCurve=DataInterface() {propertyValues {Curve {keys [
          <at 0 1 RCIM_Cubic>
          {arriveTangentWeight=0.0599, interpMode=RCIM_Cubic, leaveTangentWeight=0.3128, tangentMode=RCTM_User, time=0.1606, value=0.7431}
          <at 1 1.5>
         ]}}})
    )) {isSetParameters=true}
   &ParticleState()
   &ScaleColor(
    Scale RGB=&VectorFromCurve(VectorCurve=DataInterface() {propertyValues {
       XCurve {keys [<at 0.15 0.2031 RCIM_Cubic>, {arriveTangent=-0.0069, interpMode=RCIM_Cubic, leaveTangent=-0.0069, time=0.4, value=0.4554}, <at 0.6 0.2>]}
       YCurve {keys [
         <at 0.15 0.9529 RCIM_Cubic>
         {arriveTangent=-1.9053, interpMode=RCIM_Cubic, leaveTangent=-1.9053, tangentMode=RCTM_User, time=0.4, value=0.4034}
         <at 0.6 0.2>
        ]}
       ZCurve {keys [
         <at 0.15 1.5 RCIM_Cubic>
         {arriveTangent=-2.8653, interpMode=RCIM_Cubic, leaveTangent=-2.8653, tangentMode=RCTM_User, time=0.4, value=1}
         <at 0.6 0.2>
        ]}
      }})
    ScaleA=true
    ScaleRGB=true
   )
   &DynamicMaterialParameters(
    Index 0 Param 1=&FloatFromCurve(FloatCurve=DataInterface() {propertyValues {Curve {keys [
         {arriveTangent=10.2316, interpMode=RCIM_Cubic, leaveTangent=10.2316, tangentMode=RCTM_User, time=0, value=0}
         {arriveTangentWeight=0.1503, interpMode=RCIM_Cubic, leaveTangentWeight=0.1107, tangentMode=RCTM_User, time=0.1, value=1}
         {arriveTangent=-0.7716, interpMode=RCIM_Cubic, leaveTangent=-0.7716, time=0.1736, value=0.7636}
         {arriveTangent=-0.924, interpMode=RCIM_Cubic, leaveTangent=-0.924, time=0.8596, value=0.4139}
         {arriveTangent=-0.9141, arriveTangentWeight=0.0468, interpMode=RCIM_Cubic, leaveTangent=-0.9141, tangentMode=RCTM_User, time=1, value=0}
        ]}}})
    Index 0 Param 2=&FloatFromCurve(FloatCurve=DataInterface() {propertyValues {Curve {keys [<at 0 1 RCIM_Cubic>,<at 0.5 0>]}}})
    Index 0 Param 3=-0.25
    Index 0 Param 4=1
    Module.Index0Param0Write=true
    Module.Index0Param1Write=true
    Module.Index0Param2Write=true
    Module.Index0Param3Write=true
   )
   &Drag(Drag=2, Use Linear Drag=true, Use Rotational Drag=true)
   &SolveForcesAndVelocity(Clamp Velocity=false, Limit Acceleration=false)
  ]
  rendererClasses [&NiagaraRibbonRendererProperties]
  renderers [{rendererClass=&NiagaraRibbonRendererProperties}]
 }}
summary {
 systemName=NS_Jolly_SmallObject_Falling
 emitters [{emitterName=SmokeTrail, rendererClasses [&NiagaraRibbonRendererProperties]}]
 userVariables []
}
