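# Masked, two-sided master material for Jolly Rascal falling objects, with parameterized diffuse, normal and mask textures defaulting to a pirate ship anchor.
# A timed wobble starting at WobbleStartTime eases out over WobbleDuration, scaling actor-relative offsets per axis with sphere-masked strength and delay.
# World Z against RiftHeight plus offset drives a two-color emissive gradient with fresnel and hard edge glow, plus a height-difference term suited to clipping at the rift.
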Material /Game/Athena/Items/Consumables/JollyRascal/Material/M_Jolly_FallingObject_Master
graph {
 properties {blendMode=BLEND_Masked, pixelDepthOffsetMode=PDOM_Legacy, twoSided=true, usedShadingModels=Default Lit}
 nodes [
  Divide(A=Time() {bOverride_Period=true, period=0.5}) {constB=0.5}
  n1 = SphereMask(A=ComponentMask(None=WorldPosition().XYZ) {g=true, r=true}, B=ComponentMask(None=ObjectPositionWS()) {g=true, r=true}, Radius=ObjectRadius()) {hardnessPercent=0}
  n2 = Time() {period=5}
  n3 = ScalarParameter() {parameterName=WobbleStartTime}
  n4 = Clamp(None=Divide(A=Subtract(A=@n2, B=@n3), B=ScalarParameter() {defaultValue=0.4, group=Timestamp, parameterName=WobbleDuration}))
  n5 = Saturate(None=Power(Base=@n4, Exp=ScalarParameter() {defaultValue=0.5, parameterName=EaseOutPower}) {constExponent=0.2})
  n6 = LinearInterpolate(Alpha=@n5) {constA=1, constB=0}
  n7 = Subtract(A=ActorPositionWS(), B=WorldPosition().XYZ)
  n8 = ComponentMask(None=@n7) {g=true, r=true}
  n9 = Add(
   A=Multiply(A=OneMinus(None=@n1), B=ScalarParameter() {defaultValue=-0.05, parameterName=OuterWobbleTimeOffset}) {constB=0.1}
   B=Power(Base=@n4) {constExponent=1.5}
  )
  n10 = LinearInterpolate(
   A=VectorParameter() {parameterName=WobbleMult1, defaultValue=<rgba 1 1 0 1>}.RGB
   B=VectorParameter() {parameterName=WobbleMult2, defaultValue=<rgba 0 0 2 1>}.RGB
   Alpha=ConstantBiasScale(None=Sine(None=@n9) {period=0.75})
  )
  n11 = Sine(None=@n5)
  n12 = LinearInterpolate(Alpha=ConstantBiasScale(None=OneMinus(None=@n11))) {constA=1, constB=0}
  n13 = Multiply(
   A=AppendVector(A=Multiply(A=@n8) {constB=-1}, B=ComponentMask(None=@n7) {b=true})
   B=Add(A=VectorParameter() {parameterName=ScaleAxis, defaultValue=<rgba 0.5 0.5 1 1>}.RGB, B=Multiply(A=@n10, B=@n12))
  )
  Multiply(
   A=Multiply(A=LinearInterpolate(Alpha=@n1) {constA=0.7}, B=Multiply(A=@n6, B=@n13) {constA=1}) {constA=1}
   B=ScalarParameter() {defaultValue=0.4, parameterName=WobbleStrength}
  )
  n14 = VectorParameter() {parameterName=Color2, defaultValue=<rgba 0.8304 0 1 0>}
  n15 = VectorParameter() {parameterName=Color1, defaultValue=<rgba 0 4.1939 5 0>}
  n16 = ComponentMask(None=WorldPosition().XYZ) {b=true}
  n17 = ScalarParameter() {defaultValue=1200, parameterName=GradientSize}
  n18 = Add(A=@n16, B=Multiply(A=@n17) {constB=0.5})
  n19 = Add(A=ScalarParameter() {parameterName=RiftHeight}, B=ScalarParameter() {defaultValue=-100, parameterName=RiftHeightOffset})
  n20 = Saturate(None=Divide(A=Subtract(A=@n18, B=@n19), B=@n17))
  n21 = Power(Base=@n20) {constExponent=7}
  n22 = Multiply(A=@n20, B=ScalarParameter() {defaultValue=1.015, parameterName=GlowEdgeSize}) {constB=1.25}
  n23 = Add(A=Multiply(A=Fresnel() {baseReflectFraction=0, exponent=4}, B=@n20), B=Add(A=Multiply(A=@n21) {constB=10}, B=Round(None=@n22)))
  n24 = Saturate(None=@n23)
  Multiply(
   A=ScalarParameter() {defaultValue=5, parameterName=EmissiveBrightness}
   B=Multiply(A=LinearInterpolate(A=@n14.RGB, B=@n15.RGB, Alpha=@n24), B=@n23)
  ) {constB=50}
  Subtract(A=@n19, B=@n16)
  TextureSampleParameter2D() {parameterName=Diffuse, texture=/Game/Environments/Sets/Pirate_Cove/Pirate_Ship/Props/Textures/T_Pirate_Ship_Anchor_D}
  TextureSampleParameter2D() {
   parameterName=Normal
   samplerType=SAMPLERTYPE_Normal
   texture=/Game/Environments/Sets/Pirate_Cove/Pirate_Ship/Props/Textures/T_Pirate_Ship_Anchor_N
  }
  TextureSampleParameter2D() {
   parameterName=Masks
   samplerType=SAMPLERTYPE_Masks
   texture=/Game/Environments/Sets/Pirate_Cove/Pirate_Ship/Props/Textures/T_Pirate_Ship_Anchor_S
  }
 ]
}
