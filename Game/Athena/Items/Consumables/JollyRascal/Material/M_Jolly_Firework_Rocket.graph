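# Lit mesh-particle material for the Jolly Rascal firework rocket, with emissive glow and animated vertex wobble.
# Emissive adds a two-color Fresnel rim to a hot glow, shaped by diffuse blue channel and UV1 gradient, then tints by particle color.
# Time-driven sines at differing periods create XYZ sway, weighted toward the end opposite the glow, plus a pulse along normals.
# Dynamic parameters let particles scale wobble amount and speed.
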Material /Game/Athena/Items/Consumables/JollyRascal/Material/M_Jolly_Firework_Rocket
graph {
 properties {bUsedWithMeshParticles=true, bUsedWithNiagaraMeshParticles=true, pixelDepthOffsetMode=PDOM_Legacy, usedShadingModels=Default Lit}
 nodes [
  n1 = Fresnel() {baseReflectFraction=0, exponent=3}
  n2 = Multiply(
   A=LinearInterpolate(
    A=VectorParameter() {parameterName=Color2, defaultValue=<rgba 0.8304 0 1 0>}.RGB
    B=VectorParameter() {parameterName=Color1, defaultValue=<rgba 0 4.1939 5 0>}.RGB
    Alpha=@n1
   )
   B=@n1
  )
  n3 = ComponentMask(None=TextureCoordinate() {coordinateIndex=1}) {g=true}
  n4 = Power(Base=@n3) {constExponent=8}
  n5 = Multiply(
   A=Add(A=TextureSample() {texture=/Game/Athena/Items/Consumables/JollyRascal/Textures/T_Jollyrascal_D}.B) {constB=0.2}
   B=Multiply(A=VectorParameter() {parameterName=HotColor, defaultValue=<rgba 3 0.5392 0 0>}.RGB, B=Saturate(None=@n4))
  )
  MaterialFunctionCall(Emissive=Multiply(A=ParticleColor().RGB, B=Add(A=@n2, B=@n5))) {materialFunction=&MF_FX_EmissiveAdjustments}
  n6 = OneMinus(None=@n3)
  n7 = DynamicParameter() {paramNames [Wobble,WobbleSpeed,Param3,Param4]}
  n8 = Multiply(A=Time(), B=Multiply(A=ScalarParameter() {defaultValue=1, parameterName=WobbleSpeed}, B=@n7.WobbleSpeed))
  n9 = Sine(None=@n8) {period=0.27}
  n10 = Sine(None=@n8) {period=0.2}
  n11 = VectorParameter() {parameterName=WobbleAmount, defaultValue=<rgba 1 1 2 0.75>}
  n12 = Multiply(A=AppendVector(A=AppendVector(A=@n9, B=@n10), B=Sine(None=@n8) {period=0.12}), B=@n11.RGB)
  n13 = Multiply(A=Abs(None=Sine(None=@n8) {period=0.21}), B=VertexNormalWS())
  Multiply(A=Add(A=Multiply(A=@n6, B=@n12), B=Multiply(A=@n13, B=@n11.A)), B=@n7.Wobble)
  TextureSample() {samplerType=SAMPLERTYPE_Masks, texture=/Game/Athena/Items/Consumables/JollyRascal/Textures/T_Jollyrascal_S}
  TextureSample() {samplerType=SAMPLERTYPE_Normal, texture=/Game/Athena/Items/Consumables/JollyRascal/Textures/T_Jollyrascal_N}
 ]
}
